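; Blueprint for the Gifted Traveler orb prop, a building actor that becomes invulnerable at runtime.
; On BeginPlay it waits asynchronously for the current playlist data to be ready, then sets bIsInvulnerable to true.
; An OnRep_OrbDamaged replication handler and the construction script are present but empty.
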(event EventBeginPlay
  ; Init
  (Class|FortAsyncActionCurrentPlaylistReady|PlaylistDataReadyAsync
    (:OnReady
      (Variables|BuildingActor|SetbIsInvulnerable true))))

(fn OnRep_OrbDamaged)

(fn ConstructionScript)
