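; Invulnerable collision actor for the Gifted Traveler orb that repels players, pickups and vehicles.
; Once playlist data is ready, the server binds damage: sphere hits stamp "Hit Time" custom primitive data on the Orb mesh, and instigators near the Mask Maker get knockback and sound.
; Authority-only overlaps apply environment damage in the car box.
; Impulse volumes give non-NPC players a knockback GameplayEffect plus orb hit VFX, and toss pickups.
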(event EventBeginPlay
  ; Damage Event
  (Class|FortAsyncActionCurrentPlaylistReady|PlaylistDataReadyAsync
    (:OnReady
      (Variables|BuildingActor|SetbIsInvulnerable true)
      (if (Networking|IsServer)
        (EventDispatchers|BindEventtoOnDamaged (delegate Custom|OnDamaged_Event))))))

(event Custom|OnDamaged_Event
  (params (DamagedActor :type Actor%20Object%20Reference) (Damage :type Float%20%28single-precision%29) (InstigatedBy :type Controller%20Object%20Reference) (DamageCauser :type Actor%20Object%20Reference) (HitLocation :type Vector) (FHitComponent :type Primitive%20Component%20Object%20Reference) (BoneName :type Name) (Momentum :type Vector))
  ; Damage Event
  (if (== FHitComponent (Variables|Default|GetSphere%20-%20Damage%20-%20Orb))
    (CallFunction|OrbDamageVFX DamageCauser))
  (bind _asfort_player_controller (Utilities|Casting|CastToFortPlayerController InstigatedBy))
  (if (<= (Math|Vector|Distance(Vector) (Transformation|GetActorLocation (Game|Player|GetPlayerPawn _asfort_player_controller)) (Variables|Default|GetMask%20Maker%20World%20Location)) 3.5e+02)
    (GameplayEffects|FortApplyGameplayEffectToSelf :self (Ability|GetFortAbilitySystemComponent (Game|Player|GetPlayerPawn _asfort_player_controller)) :GameplayEffectClass "/GiftedTraveler/Assets/Props/Orb/GameplayCues/GE_MM_GrantKnockback.GE_MM_GrantKnockback_C" :OptionalSourceObject self :OptionalEffectCauser self :OptionalDamageSource DamageCauser :Level 0.0)
    (Audio|PlaySoundatLocation (Variables|Default|GetKnockback%20Sound) (Transformation|GetActorLocation (Game|Player|GetPlayerPawn _asfort_player_controller)))))

(event Custom|OrbDamageVFX
  (params (Instigator :type Actor%20Object%20Reference))
  (bind _orb (Variables|Default|GetOrb))
  ; Damage VFX on the orb
  (Utilities|IsValid _orb
    (:Is%20Valid
      (Rendering|Material|SetScalarParameterforCustomPrimitiveData (Variables|/Script/FortniteGame.BuildingSMActor|GetStaticMeshComponent _orb) "Hit Time" (Utilities|Time|GetGameTimeinSeconds)))))

(event OnComponentBeginOverlap(Box-cardamage)
  (params (OverlappedComponent :type Primitive%20Component%20Object%20Reference) (OtherActor :type Actor%20Object%20Reference) (OtherComp :type Primitive%20Component%20Object%20Reference) (OtherBodyIndex :type Integer) (bFromSweep :type Boolean) (SweepResult :type Hit%20Result%20Structure%20%28by%20ref%29))
  ; Blow Up Vehicles
  (Networking|SwitchHasAuthority
    (:Authority
      (Environment|ApplyGlobalEnvironmentDamagetoActor OtherActor 5e+02))))

(event OnComponentBeginOverlap(ImpulseMaskMaker)
  (params (OverlappedComponent :type Primitive%20Component%20Object%20Reference) (OtherActor :type Actor%20Object%20Reference) (OtherComp :type Primitive%20Component%20Object%20Reference) (OtherBodyIndex :type Integer) (bFromSweep :type Boolean) (SweepResult :type Hit%20Result%20Structure%20%28by%20ref%29))
  ; Impulse
  (Networking|SwitchHasAuthority
    (:Authority
      (CallFunction|PushActorAway OtherActor))))

(event OnComponentBeginOverlap(ImpulseBeam)
  (params (OverlappedComponent :type Primitive%20Component%20Object%20Reference) (OtherActor :type Actor%20Object%20Reference) (OtherComp :type Primitive%20Component%20Object%20Reference) (OtherBodyIndex :type Integer) (bFromSweep :type Boolean) (SweepResult :type Hit%20Result%20Structure%20%28by%20ref%29))
  ; Impulse
  (Networking|SwitchHasAuthority
    (:Authority
      (CallFunction|PushActorAway OtherActor))))

(event OnComponentBeginOverlap(ImpulseBeamSphere)
  (params (OverlappedComponent :type Primitive%20Component%20Object%20Reference) (OtherActor :type Actor%20Object%20Reference) (OtherComp :type Primitive%20Component%20Object%20Reference) (OtherBodyIndex :type Integer) (bFromSweep :type Boolean) (SweepResult :type Hit%20Result%20Structure%20%28by%20ref%29))
  ; Impulse
  (Networking|SwitchHasAuthority
    (:Authority
      (CallFunction|PushActorAway OtherActor))))

(fn OnRep_OrbDamaged)

(fn PushActorAway
  (params (Actor :type Actor%20Object%20Reference))
  (bind _asfort_player_pawn_athena (Utilities|Casting|CastToFortPlayerPawnAthena Actor)
    (:then
      (if (not (GameplayTags|HasMatchingGameplayTag _asfort_player_pawn_athena "(TagName=\"Pawn.Athena.NPC\")"))
        (GameplayEffects|FortApplyGameplayEffectToSelf :self (Ability|GetFortAbilitySystemComponent _asfort_player_pawn_athena) :GameplayEffectClass "/GiftedTraveler/Assets/Props/Orb/GameplayCues/GE_MM_GrantKnockback.GE_MM_GrantKnockback_C" :OptionalSourceObject self :OptionalEffectCauser self :Level 0.0)
        (CallFunction|OrbDamageVFX 0)))
    (:CastFailed
      (bind _asfort_pickup (Utilities|Casting|CastToFortPickup Actor))
      (bind (_returnvalue_roll _returnvalue_pitch _returnvalue_yaw) (Math|Rotator|BreakRotator (Math|Rotator|FindLookatRotation (Transformation|GetActorLocation) (Transformation|GetActorLocation _asfort_pickup))))
      (Pickup|TossPickupinDirection _asfort_pickup (* (Math|Vector|MakeVector _returnvalue_roll _returnvalue_pitch (Math|Float|Absolute(Float) _returnvalue_yaw)) 5.0) 0 false true))))

(fn ConstructionScript)
